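; Zero Point vortex effect actor that adds an on/off gate over the behavior inherited from its parent, driven by the EnabledHF scalable float.
; BeginPlay sets isDisabled when the flag is false; otherwise it runs the parent logic and waits for authority match ready.
; On match ready it checks the flag again and disables the actor if it reads false.
; The construction script runs the parent's, then zeroes the S_FX_GeyserCone component's relative rotation.
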(event EventBeginPlay
  (bind _enabledhf (Variables|Default|GetEnabledHF))
  (bind _returnvalue (ScalableFloat|GetValueasBool _enabledhf))
  (if _returnvalue
    (|CallParentFunction)
    (Class|FortAsyncActionAuthorityMatchReady|MatchReadyAsync
      (:OnReady
        (if (not _returnvalue)
          (Variables|Default|SetisDisabled true))))
    (else
      (Variables|Default|SetisDisabled true))))

(fn ConstructionScript
  (|CallParentFunction)
  (Transformation|SetRelativeRotation (Variables|Default|GetS_FX_GeyserCone) (Math|Rotator|MakeRotator 0.0) false false))
