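# Material for a paper coffee cup mesh, set up for use with Niagara mesh particles as a Default Lit effect asset.
# An offset SphereMask on scaled UVs blends two constant tints, which multiply the diffuse texture, then multiplies that result again as a soft falloff.
# The same mask scales a Roughness parameter and feeds an inverted copy, alongside a normal map sample.
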Material /Game/Environments/Hera/FX/Materials/M_CoffeeCup_Paper
graph {
 properties {
  bAutomaticallySetUsageInEditor=false
  bUsedWithNiagaraMeshParticles=true
  previewMesh=/Game/Environments/Hera/FX/Meshes/SM_CoffeeCup_Paper_01
  usedShadingModels=Default Lit
 }
 nodes [
  n1 = Constant3Vector() {constant=<rgba 0.9381 0.4532 1 1>}
  n2 = Constant3Vector() {constant=<rgba 0 0.2626 1 1>}
  n3 = SphereMask(A=TextureCoordinate() {uTiling=1.25, vTiling=1.25}, B=Constant2Vector() {g=0.75, r=0.5}) {attenuationRadius=0.5, hardnessPercent=0.5}
  Multiply(
   A=Multiply(A=LinearInterpolate(A=@n1, B=@n2, Alpha=@n3), B=TextureSample() {texture=/Game/Environments/Hera/FX/Textures/T_CoffeCup_Paper_01_D}.RGB)
   B=@n3
  )
  Multiply(A=ScalarParameter() {defaultValue=1, parameterName=Roughness}, B=@n3)
  OneMinus(None=@n3)
  TextureSample() {samplerType=SAMPLERTYPE_Normal, texture=/Game/Environments/Hera/FX/Textures/T_CoffeCup_Paper_01_N}
 ]
}
