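# Unlit, two-sided translucent glow material that fakes a volumetric light bloom on a camera-facing mesh.
# World position offset math turns vertices toward the view around the component pivot and pushes them toward the camera.
# A powered UV sphere mask flickers through two sines, phased by pivot offset from the actor, and tinted by a hue-shiftable colour.
# Opacity adds near-camera fade, plus depth fade only on selected quality levels and deferred shading.
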Material /Game/Environments/Hera/FX/Materials/M_FakeVolumeGlow
graph {
 properties {blendMode=BLEND_Translucent, pixelDepthOffsetMode=PDOM_Legacy, shadingModel=MSM_Unlit, twoSided=true, usedShadingModels=Unlit}
 nodes [
  n1 = MaterialFunctionCall() {materialFunction=&LocalPosition}
  n2 = MaterialFunctionCall() {materialFunction=/Engine/Functions/Engine_MaterialFunctions02/WorldPositionOffset/ObjectScale}
  n3 = MaterialFunctionCall!(
   Vector1=Constant3Vector() {constant=<rgba 1 0 0 0>}
   Vector2=Constant3Vector() {constant=<rgba 0 1 0 0>}
   Vector3=Constant3Vector() {constant=<rgba 0 0 1 0>}
  ) {materialFunction=/Engine/Functions/Engine_MaterialFunctions02/Math/MakeVectorsOrthogonal}
  n4 = Transform(None=MaterialFunctionCall(
    VectorToTransform=Multiply(A=@n1."Local Position (Excluding Offsets)", B=@n2.Scale XYZ)
    BasisX=@n3.Vector1
    BasisY=@n3.Vector1
    BasisZ=@n3.Vector1
   ) {materialFunction=&Transform3x3Matrix}.Transformed Vector) {transformSourceType=TRANSFORMSOURCE_View}
  n5 = WorldPosition()
  n6 = Subtract(A=CameraPositionWS(), B=WorldPosition().XYZ)
  Add(
   A=Add(
    A=Subtract(A=@n4, B=@n5.XYZ)
    B=MaterialFunctionCall() {materialFunction=&ComponentPivotLocation}.Result
   )
   B=Multiply(A=Normalize(VectorInput=@n6), B=ScalarParameter() {defaultValue=20, parameterName=CameraOffsetAmount})
  ) {constB=0}
  Constant3Vector() {constant=<rgba 0 0 -1 0>}
  n7 = VectorParameter() {parameterName=EmissiveColour, defaultValue=<rgba 5 1.6633 0.45 0.035>}
  MaterialFunctionCall(Hue Shift Percentage=ScalarParameter() {parameterName="Hue Shift Percentage (S)"}, Texture=@n7.RGB) {materialFunction=&HueShift}
  n8 = MaterialFunctionCall(Float3=Subtract(
    A=MaterialFunctionCall() {materialFunction=&ComponentPivotLocation}.Result
    B=ActorPositionWS()
   )) {materialFunction=&BreakOutFloat3Components}
  n9 = Multiply(A=Add(A=Add(A=@n8.R, B=@n8.R), B=@n8.B)) {constB=0.05}
  Multiply(A=@n9) {constB=0.7}
  n10 = TextureCoordinate()
  n11 = Constant() {r=0.5}
  n12 = Constant() {r=0.5}
  n13 = Constant()
  n14 = Add(A=Time(), B=@n9)
  n15 = Sine(None=Multiply(A=@n14) {constB=1.351})
  n16 = Sine(None=Multiply(A=@n14) {constB=1.21})
  n17 = Multiply(A=Add(A=Multiply(A=@n15) {constB=0.025}, B=Multiply(A=@n16) {constB=0.04})) {constB=0.5}
  n18 = Power(Base=Multiply(A=SphereMask(A=@n10, B=@n11, Radius=@n12, Hardness=@n13), B=Add(A=@n17))) {constExponent=6}
  n19 = DepthFade(Opacity=@n18, FadeDistance=ScalarParameter() {defaultValue=50, group=Opacity, parameterName=Depth Fade Distance, sortPriority=6})
  n20 = QualitySwitch(Default=@n19, Low=@n18, High=@n19, Medium=@n18, Epic=@n19)
  n21 = ScalarParameter() {defaultValue=500, parameterName=NearFade}
  n22 = ScalarParameter() {defaultValue=250, parameterName=NearFadeGone}
  Multiply(
   A=Multiply(A=@n7.A, B=ShadingPathSwitch(Default=@n20, Deferred=@n20, Forward=@n18, Mobile=@n18))
   B=MaterialFunctionCall(Distance=Subtract(A=@n21, B=@n22)) {materialFunction=&Near_Camera_Fade}.Result
  )
 ]
}
