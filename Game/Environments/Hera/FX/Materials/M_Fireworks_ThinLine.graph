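# Unlit, masked, two-sided material for Niagara mesh particles, drawing thin firework lines with particle-colored emissive.
# SplineThicken sets line width, using WidthBase for base and tip, and adds time-scrolled vector noise for curling world position offset.
# Two panned cell-noise samples multiply, get contrast, then combine with a U-gradient end fade and particle alpha to break up the line.
# Dynamic parameters drive width, curl, noise offsets and contrast per particle.
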Material /Game/Environments/Hera/FX/Materials/M_Fireworks_ThinLine
graph {
 properties {
  bAutomaticallySetUsageInEditor=false
  bUsedWithNiagaraMeshParticles=true
  blendMode=BLEND_Masked
  shadingModel=MSM_Unlit
  twoSided=true
  usedShadingModels=Unlit
 }
 nodes [
  n1 = DynamicParameter() {paramNames [WidthBase,WidthTip,U_Offset,Y_Offset]}
  n2 = WorldPosition()
  n3 = ObjectPositionWS()
  n4 = DynamicParameter() {parameterIndex=1, paramNames [Curl, CurlIntensity, CurlSpeed, Y_Offset_2ND]}
  n5 = Time()
  n6 = Divide(A=Subtract(A=Subtract(A=@n2.XYZ, B=@n3), B=Multiply(A=@n4.CurlSpeed, B=@n5)), B=@n4.Curl)
  Add(
   A=MaterialFunctionCall(WidthBase=@n1.WidthBase, WidthTip=@n1.WidthBase) {materialFunction=&SplineThicken}.WorldPositionOffset
   B=Multiply(A=VectorNoise(World Position=@n6) {noiseFunction=VNF_VectorALU}, B=@n4.CurlIntensity)
  )
  Add()
  DynamicParameter() {paramNames [WidthBase,WidthTip,U_Offset,Y_Offset]}
  n7 = ParticleColor()
  MaterialFunctionCall(Emissive=@n7.RGB) {materialFunction=&MF_FX_EmissiveAdjustments}
  n8 = MaterialFunctionCall(UV Channel=TextureCoordinate()) {materialFunction=&LinearGradient}
  n9 = Multiply(A=OneMinus(None=@n8.UGradient), B=@n8.UGradient)
  n10 = TextureCoordinate()
  n11 = DynamicParameter() {paramNames [WidthBase,WidthTip,U_Offset,Y_Offset]}
  n12 = AppendVector(A=@n11.U_Offset, B=@n11.U_Offset)
  n13 = ScalarParameter() {defaultValue=2, parameterName=U_Tile}
  n14 = ScalarParameter() {defaultValue=64, parameterName=V_Tile}
  n15 = Panner(Coordinate=Multiply(A=Add(A=@n10, B=@n12), B=AppendVector(A=@n13, B=@n14))) {speedY=-1}
  n16 = TextureCoordinate()
  n17 = DynamicParameter() {parameterIndex=2, defaultValue=<rgba 1 512 1 1>, paramNames [X_Tile_2ND,Y_Tile_2ND,U_Offset_2ND,Y_Offset_2ND]}
  n18 = AppendVector(A=@n17.U_Offset_2ND, B=@n17.U_Offset_2ND)
  n19 = Panner(Coordinate=Multiply(A=Add(A=@n16, B=@n18), B=AppendVector(A=@n17.X_Tile_2ND, B=@n17.X_Tile_2ND))) {speedY=-1}
  n20 = Subtract(A=Multiply(
    A=TextureSample(UVs=@n15) {texture=/Game/Global/Textures/Default/Noise/T_CellNoise_01}.G
    B=TextureSample(UVs=@n19) {texture=&T_CellNoise_01}.G
   )) {constB=0.1}
  n21 = DynamicParameter() {parameterIndex=3, paramNames [Contrast,Param2,Param3,Param4]}
  Multiply(
   A=Multiply(
    A=Multiply(A=@n9) {constB=3}
    B=MaterialFunctionCall(In=@n20, Contrast=@n21.Contrast) {materialFunction=&CheapContrast}.Result
   )
   B=@n7.A
  )
  ScalarParameter() {defaultValue=2, parameterName=U_Tile}
  ScalarParameter() {defaultValue=250, parameterName="Contrast (S)"}
  ScalarParameter() {defaultValue=1, group=Movement, parameterName=Curl_Speed}
  ScalarParameter() {defaultValue=1, group=Movement, parameterName=Curl_Scale}
  ScalarParameter() {defaultValue=1, group=Movement, parameterName=Curl_Intensity}
  ScalarParameter() {defaultValue=64, parameterName=V_Tile}
 ]
}
