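# Masked lit material for rubber duck debris, built for Niagara mesh particles, using the duck's color, normal and mask textures.
# A dynamic parameter drives a per-particle hue shift of the color texture, reduced to zero inside a spherical mask over scaled UVs so that region keeps its original color.
# A second particle parameter feeds a textureless Bayer dither, and the graph also contains a near-camera fade function.
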Material /Game/Environments/Hera/FX/Materials/M_FloatingDebris_Duck
graph {
 properties {bAutomaticallySetUsageInEditor=false, bUsedWithNiagaraMeshParticles=true, blendMode=BLEND_Masked, usedShadingModels=Default Lit}
 nodes [
  n1 = DynamicParameter() {paramNames [HueShift,Param2,Param3,Param4]}
  n2 = Constant() {r=1}
  n3 = Constant() {r=1.25}
  n4 = SphereMask(A=Multiply(A=TextureCoordinate(), B=AppendVector(A=@n2, B=@n3)), B=Constant2Vector() {g=0.4}) {attenuationRadius=0.25}
  n5 = TextureSample() {texture=/Game/Environments/Sets/Summer/Props/Textures/T_rubber_duck_shades_D}
  MaterialFunctionCall(Hue Shift Percentage=LinearInterpolate(A=@n1.HueShift, Alpha=ComponentMask(None=@n4) {g=true}) {constB=0}, Texture=@n5.RGB) {materialFunction=&HueShift}
  MaterialFunctionCall!(In=@n1.Param2, Iterations, Scale) {materialFunction=/Game/Packages/Fortress_SharedMaterials/Base_Material_Functions/MF_DitherBayerNoTex}
  MaterialFunctionCall() {materialFunction=&Near_Camera_Fade}
  Multiply(A=@n4, B=@n5.RGB)
  Multiply(A=@n1.HueShift, B=OneMinus(None=@n4))
  TextureSample() {samplerType=SAMPLERTYPE_Normal, texture=/Game/Environments/Sets/Summer/Props/Textures/T_rubber_duck_shades_N}
  TextureSample() {samplerType=SAMPLERTYPE_Masks, texture=/Game/Environments/Sets/Summer/Props/Textures/T_rubber_duck_shades_S}
 ]
}
