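# Default Lit metallic balloon material for theme park effects, usable on Niagara mesh particles, with particle color, flattened normal map and metallic/roughness parameters.
# Its vertex motion math sums a slow whole-object wobble around the actor position, on a time-varying axis, with per-element wind rotation.
# The wind term rotates around Pivot Painter pivots on a local axis, using two sine waves phase-shifted by per-element random values and masked by vertex color green.
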Material /Game/Environments/Hera/FX/Materials/M_ThemePark_Balloon_01_Metallic
graph {
 properties {bAutomaticallySetUsageInEditor=false, bUsedWithNiagaraMeshParticles=true, usedShadingModels=Default Lit}
 nodes [
  n1 = Multiply(A=Time(), B=ScalarParameter() {defaultValue=0.1714, parameterName=SpeedBallon_2})
  n2 = AppendVector(A=Sine(None=@n1), B=AppendVector(A=Sine(None=@n1), B=Cosine(None=@n1)))
  n3 = Sine(None=Multiply(A=Time(), B=ScalarParameter() {defaultValue=0.0892, parameterName=OverallRotationSpeed}))
  n4 = ScalarParameter() {defaultValue=0.01, parameterName=OverallRotation}
  n5 = Constant3Vector()
  n6 = ActorPositionWS()
  n7 = WorldPosition() {worldPositionShaderOffset=WPT_ExcludeAllShaderOffsets}
  n8 = Normalize(VectorInput=Transform(None=VectorParameter() {parameterName=RotaionAxis, defaultValue=<rgba 0 0 1 1>}.RGB) {transformSourceType=TRANSFORMSOURCE_Local})
  n9 = Time()
  n10 = MaterialFunctionCall() {materialFunction=/Engine/Functions/Engine_MaterialFunctions02/PivotPainter/PivotPainter_PerObjectData}
  n11 = ScalarParameter() {defaultValue=1.5613, parameterName=OffsetVariation}
  n12 = Multiply(
   A=Add(A=@n9, B=Multiply(A=@n10.Random Value Per Element, B=@n11))
   B=ScalarParameter() {defaultValue=0.444, parameterName=WindRotation_Speed_1}
  )
  n13 = ScalarParameter() {defaultValue=2.2338, parameterName=OffsetVariation_1}
  n14 = Multiply(
   A=Add(A=Multiply(A=@n10.Random Value Per Element, B=@n13), B=@n9)
   B=ScalarParameter() {defaultValue=0.469, parameterName=WindRotation_Speed}
  )
  n15 = Multiply(A=Add(A=Sine(None=@n12), B=Sine(None=@n14)), B=ScalarParameter() {defaultValue=0.02, parameterName=WindRotation_strength})
  Add(
   A=RotateAboutAxis(NormalizedRotationAxis=Normalize(VectorInput=@n2), RotationAngle=Multiply(A=@n3, B=@n4), PivotPoint=Add(A=@n5, B=@n6), Position=@n7.XYZ)
   B=Multiply(A=RotateAboutAxis(NormalizedRotationAxis=@n8, RotationAngle=@n15, PivotPoint=@n10.Pivot Position, Position=@n7.XYZ), B=VertexColor().G)
  )
  Constant3Vector() {constant=<rgba 1 0.0353 0 1>}
  MaterialFunctionCall(
   Normal=TextureSampleParameter2D() {parameterName=Normal, samplerType=SAMPLERTYPE_Normal, texture=/Game/Environments/Hera/FX/Textures/T_WonkeeLand_Balloon_01_N}.RGB
   Flatness=ScalarParameter() {defaultValue=1, parameterName="Flatness (S)"}
  ) {materialFunction=&FlattenNormal}
  Normalize()
  ParticleColor()
  ScalarParameter() {defaultValue=1, parameterName=Metallic}
  ScalarParameter() {parameterName=Roughness}
  ScalarParameter()
  VectorParameter() {parameterName=RotationCenter, defaultValue=<rgba 0 0 0 1>}
  WorldPosition()
 ]
}
