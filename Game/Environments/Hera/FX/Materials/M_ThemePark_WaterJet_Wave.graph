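# M_ThemePark_WaterJet_Wave is a material in the Hera environment FX folder, named for a theme park water jet wave effect.
# Its graph contains no expressions, so the file shows no shading logic, textures or parameters.
# It uses the Default Lit shading model and has automatic setting of usage flags in the editor turned off.
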Material /Game/Environments/Hera/FX/Materials/M_ThemePark_WaterJet_Wave
graph {expressions {}, properties {bAutomaticallySetUsageInEditor=false, usedShadingModels=Default Lit}}
