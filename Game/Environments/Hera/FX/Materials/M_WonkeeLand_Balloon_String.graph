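# Two-sided lit material for WonkeeLand balloon strings, usable on Niagara mesh particles, built around vertex-offset animation.
# SplineThicken sets string width, added to a slow sway around the actor position on a time-varying axis.
# Pivot Painter pivots and random values add per-element wind rotation, masked by vertex green.
# A separate branch adds UV-driven sine ripples perpendicular to the camera direction; no material output connections appear for either branch.
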Material /Game/Environments/Hera/FX/Materials/M_WonkeeLand_Balloon_String
graph {
 properties {bAutomaticallySetUsageInEditor=false, bUsedWithNiagaraMeshParticles=true, twoSided=true, usedShadingModels=Default Lit}
 nodes [
  n1 = MaterialFunctionCall(
   WidthBase=ScalarParameter() {defaultValue=25, parameterName="WidthBase (S)"}
   WidthTip=ScalarParameter() {defaultValue=25, parameterName="WidthTip (S)"}
  ) {materialFunction=&SplineThicken}
  n2 = Multiply(A=Time(), B=ScalarParameter() {defaultValue=0.1714, parameterName=SpeedBallon_2})
  n3 = Sine(None=@n2)
  n4 = Cosine(None=@n2)
  n5 = Normalize(VectorInput=AppendVector(A=Sine(None=@n2), B=AppendVector(A=@n3, B=@n4)))
  n6 = Time()
  n7 = ScalarParameter() {defaultValue=0.0892, parameterName=OverallRotationSpeed}
  n8 = Multiply(A=Sine(None=Multiply(A=@n6, B=@n7)), B=ScalarParameter() {defaultValue=0.01, parameterName=OverallRotation})
  n9 = Add(A=Constant3Vector(), B=ActorPositionWS())
  n10 = WorldPosition() {worldPositionShaderOffset=WPT_ExcludeAllShaderOffsets}
  n11 = VectorParameter() {parameterName=RotaionAxis, defaultValue=<rgba 0 0 1 1>}
  n12 = Time()
  n13 = MaterialFunctionCall() {materialFunction=/Engine/Functions/Engine_MaterialFunctions02/PivotPainter/PivotPainter_PerObjectData}
  n14 = Multiply(A=@n13.Random Value Per Element, B=ScalarParameter() {defaultValue=1.5613, parameterName=OffsetVariation})
  n15 = Sine(None=Multiply(A=Add(A=@n12, B=@n14), B=ScalarParameter() {defaultValue=0.444, parameterName=WindRotation_Speed_1}))
  n16 = Multiply(A=@n13.Random Value Per Element, B=ScalarParameter() {defaultValue=2.2338, parameterName=OffsetVariation_1})
  n17 = Sine(None=Multiply(A=Add(A=@n16, B=@n12), B=ScalarParameter() {defaultValue=0.469, parameterName=WindRotation_Speed}))
  n18 = RotateAboutAxis(
   NormalizedRotationAxis=Normalize(VectorInput=Transform(None=@n11.RGB) {transformSourceType=TRANSFORMSOURCE_Local})
   RotationAngle=Multiply(A=Add(A=@n15, B=@n17), B=ScalarParameter() {defaultValue=0.02, parameterName=WindRotation_strength})
   PivotPoint=@n13.Pivot Position
   Position=@n10.XYZ
  )
  n19 = VertexColor()
  Add(
   A=@n1.WorldPositionOffset
   B=Add(A=RotateAboutAxis(NormalizedRotationAxis=@n5, RotationAngle=@n8, PivotPoint=@n9, Position=@n10.XYZ), B=Multiply(A=@n18, B=@n19.G))
  )
  n20 = CrossProduct(
   A=MaterialFunctionCall() {materialFunction=&CameraDirectionVector}.Result
   B=Constant3Vector() {constant=<rgba 0 0 1 0>}
  )
  n21 = MaterialFunctionCall(Float2=TextureCoordinate() {vTiling=0.1}) {materialFunction=&BreakOutFloat2Components}
  n22 = Multiply(A=@n21.G) {constB=5}
  n23 = Frac(None=Multiply(A=Time()) {constB=0.5})
  n24 = Multiply(A=@n21.G) {constB=6}
  n25 = Add(A=Sine(None=Add(A=@n22, B=@n23)), B=Sine(None=Add(A=@n24, B=@n23)))
  Add(A=@n1.WorldPositionOffset, B=Multiply(A=Multiply(A=@n20, B=@n25), B=ScalarParameter() {defaultValue=5, parameterName=WPO Gain}))
  Normalize()
  VectorParameter() {parameterName=RotationCenter, defaultValue=<rgba 0 0 0 1>}
  WorldPosition()
 ]
}
