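# Masked, two-sided material for Niagara mesh particles that renders the WonkeeLand head firework as glowing sparks.
# The Sprite function turns mesh pieces into sprites, pushed along normals or radially from object center, with dynamic parameters controlling size, expansion and curl.
# Scrolling vector noise adds drift, skipped on ES2/ES3_1, which also use their own size parameter.
# Vertex color drives sine twinkle and hue shift, tinted by particle color, alongside a circular alpha mask.
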Material /Game/Environments/Hera/FX/Materials/M_WonkeeLand_Head_Firework_New
graph {
 properties {
  bEnableResponsiveAA=true
  bUsedWithNiagaraMeshParticles=true
  blendMode=BLEND_Masked
  opacityMaskClipValue=0.01
  pixelDepthOffsetMode=PDOM_Legacy
  previewMesh=/HoneyFiberEnv/FX/Meshes/SM_MP_Head_Firework_01
  twoSided=true
  usedShadingModels=Default Lit
 }
 nodes [
  Constant() {r=0.2}
  n1 = VertexColor()
  n2 = Clamp(None=@n1.A) {minDefault=0.5}
  n3 = ScalarParameter() {defaultValue=1, parameterName=ScaleSpriteSize}
  n4 = ScalarParameter() {defaultValue=2, parameterName=ScaleSpriteSize_Mobile}
  n5 = DynamicParameter() {defaultValue=<rgba 100 1 1 1>, paramNames [SpriteSize, Expansion, CurlMult, ChangeExpansion]}
  n6 = Multiply(A=FeatureLevelSwitch(Default=@n3, ES2=@n4, ES3_1=@n4, SM4_REMOVED=@n3, SM5=@n3, SM6=@n3), B=@n5.SpriteSize)
  n7 = AppendVector(A=@n6, B=@n6)
  n8 = Normalize(VectorInput=VertexNormalWS())
  n9 = Normalize(VectorInput=Subtract(A=WorldPosition().XYZ, B=ObjectPositionWS()))
  n10 = DynamicParameter() {defaultValue=<rgba 100 1 1 1>, paramNames [SpriteSize, Expansion, CurlMult, ChangeExpansion]}
  n11 = Add(
   A=MaterialFunctionCall!(Center Location=WorldPosition().XYZ, 0-1 UVs, XY scale=Multiply(A=@n2, B=@n7), OverridePivot, "Normals (Optional)") {materialFunction=&Sprite}.World position offset
   B=Multiply(A=LinearInterpolate(A=@n8, B=@n9, Alpha=@n10.ChangeExpansion), B=Multiply(A=@n1.A, B=@n5.Expansion))
  )
  n12 = Multiply(
   A=DynamicParameter() {defaultValue=<rgba 100 1 1 1>, paramNames [SpriteSize, Expansion, CurlMult, ChangeExpansion]}.CurlMult
   B=LinearInterpolate(
    A=ScalarParameter() {defaultValue=15, parameterName=Intensity_End}
    B=ScalarParameter() {defaultValue=50, parameterName=Intensity}
    Alpha=ParticleColor().A
   )
  )
  n13 = LinearInterpolate(Alpha=VertexColor().A) {constA=0.5}
  n14 = WorldPosition()
  n15 = ActorPositionWS()
  n16 = ScalarParameter() {defaultValue=15, parameterName=Speed}
  n17 = Time()
  n18 = Divide(A=Subtract(A=Subtract(A=@n14.XYZ, B=@n15), B=Multiply(A=@n16, B=@n17)), B=ScalarParameter() {defaultValue=5000, parameterName=Scale})
  n19 = Add(A=@n11, B=Multiply(A=Multiply(A=@n12, B=@n13), B=VectorNoise(World Position=@n18) {noiseFunction=VNF_VectorALU}))
  FeatureLevelSwitch(Default=@n19, ES2=@n11, ES3_1=@n11, SM4_REMOVED=@n19)
  n20 = VertexColor()
  n21 = VertexColor()
  n22 = AppendVector(A=AppendVector(A=@n21.R, B=@n21.R), B=@n21.B)
  n23 = Sine(None=Multiply(A=@n20.A, B=Time()))
  n24 = Clamp(None=VertexInterpolator(VS=Abs(None=@n23)).PS) {minDefault=0.5}
  n25 = Clamp(None=Multiply(A=ComponentMask(None=@n24) {r=true}) {constB=50})
  n26 = Multiply(
   A=MaterialFunctionCall(Hue Shift Percentage=Multiply(A=@n20.A) {constB=0.025}, Texture=Multiply(A=@n22, B=@n25)) {materialFunction=&HueShift}.Result
   B=ParticleColor().RGB
  )
  n27 = Clamp(None=@n26) {maxDefault=2}
  n28 = ShadingPathSwitch(Default=@n26, Forward=@n27, Mobile=@n27)
  MaterialFunctionCall(Emissive=Multiply(A=VertexColor(), B=FeatureLevelSwitch(Default=@n28, ES2=@n28, ES3_1=@n28, SM4_REMOVED=@n28))) {materialFunction=&MF_FX_EmissiveAdjustments}
  MaterialFunctionCall() {materialFunction=&DeriveHDRfromLDR}
  Multiply()
  n29 = TextureCoordinate()
  n30 = Constant2Vector() {g=0.5, r=0.5}
  n31 = ScalarParameter() {defaultValue=0.5, parameterName=Radius}
  n32 = ScalarParameter() {parameterName=Hardness}
  Multiply(A=Multiply(A=SphereMask(A=@n29, B=@n30, Radius=@n31, Hardness=@n32), B=@n1.A), B=ParticleColor().A)
  ParticleColor()
  SphereMask()
  SphereMask(A=TextureCoordinate(), B=Constant2Vector() {g=0.5, r=0.5}) {attenuationRadius=0.5, hardnessPercent=0}
  TextureSample() {texture=/Game/Global/Textures/Default/Masks_and_Gradients/T_LF_Octagon_Mask}
 ]
}
