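# Niagara system for an ambient butterfly environment effect in the Hera environment, built around a single ButterFlies emitter.
# SystemState is the only module recorded, and the emitter lists no renderers, so particle behavior and visuals are not visible here.
# It exposes BoxSize, Radius Mult and Spawn Mult user parameters, whose names suggest per-placement control of area and density.
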NiagaraSystem /Game/Environments/Hera/FX/NSLW_Env_ButterFlies
dependencies {usedDataInterfaces [], usedDynamicInputs [], usedModules [&SystemState]}
emitters {ButterFlies {rendererClasses [], renderers []}}
summary {
 systemName=NSLW_Env_ButterFlies
 emitters [{emitterName=ButterFlies, rendererClasses []}]
 userVariables [
  User.BoxSize {type=Vector3f, description="", defaultValue {struct=Vector3f, value=<0 0 0>}}
  User.Radius Mult {type=NiagaraFloat, defaultValue=1, description=""}
  User.Spawn Mult {type=NiagaraFloat, defaultValue=1, description=""}
 ]
}
