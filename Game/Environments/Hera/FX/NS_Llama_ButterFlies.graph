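# Niagara GPU butterfly effect that settles colored mesh and sprite butterflies on visible surfaces and scatters them near the camera.
# A short spawn burst is projected onto the depth buffer, aligned to surface normals and filtered by a GBuffer hue range.
# Resting particles do not age until camera-relative forces move them, latching an active state with curl noise, gravity and changed material parameters.
# Epic and High emitters share matching stacks with distance culling.
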NiagaraSystem /Game/Environments/Hera/FX/NS_Llama_ButterFlies
dependencies {
 usedDataInterfaces [
  &NiagaraDataInterfaceCamera
  &NiagaraDataInterfaceCurve
  &NiagaraDataInterfaceColorCurve
  &NiagaraDataInterfaceGBuffer
 ]
 usedDynamicInputs [/Game/Effects/Niagara/DynamicInputs/Math/FN_Fit]
 usedModules [
  &self:SystemSpawnScript.SystemScriptSource.SystemScriptGraph.NiagaraNodeAssignment_0.SetVariables_#1
  &SystemState
  &self:SystemSpawnScript.SystemScriptSource.SystemScriptGraph.NiagaraNodeAssignment_1.SetVariables_#2
 ]
}
emitters {
 Butterflies_Epic {
  simTarget=GPUComputeSim
  emitterUpdateScript [
   /SetVariables() {isSetParameters=true, module=SetVariables_#3}
   /SetVariables(
    Emitter.CameraForward:Vector3f=&ReturnCameraVector(Camera Property=::NewEnumerator1, Coordinate Space=::Simulation, CameraQuery=$Emitter.CameraRead)
    Emitter.CameraPosition:Vector3f=&ReturnCameraVector(Camera Property=::NewEnumerator0, Coordinate Space=::World, CameraQuery=$Emitter.CameraRead)
    Emitter.CameraRight:Vector3f=&ReturnCameraVector(Camera Property=::NewEnumerator3, Coordinate Space=::Simulation, CameraQuery=$Emitter.CameraRead)
    Emitter.CameraUp:Vector3f=&ReturnCameraVector(Camera Property=::NewEnumerator2, Coordinate Space=::Simulation, CameraQuery=$Emitter.CameraRead)
   ) {isSetParameters=true, module=SetVariables_#4}
   &EmitterState(
    Enable Distance Culling=true
    Life Cycle Mode=::NewEnumerator1
    Loop Behavior=::NewEnumerator1
    Loop Duration Mode=::NewEnumerator1
    Max Distance Response=::SleepAndClearParticles
    MaxDistance=10000
    Module.bApplySpawnCountScaleByDistance=true
    Module.bUseMaxDistance=true
    Module.bUseMinDistance=false
    Reset Age On Awaken=true
    Scalability Mode=::NewEnumerator1
    Scale Spawn Count=true
    Spawn Count Scale By Distance=DataInterface() {propertyValues {Curve {keys [<at 0.5 0>,<at 0.55 1>,<at 1 1>]}}}
    UseLoopDelay=false
   )
   &SpawnPerFrame(
    Spawn:NiagaraBool=&SetBoolByFloatComparison(A=$Emitter.Age, B=1, Comparison Type=::NewEnumerator6)
    Spawn Count:NiagaraInt32=/Niagara/DynamicInputs/Math/Min_Int(
     A:NiagaraInt32=500
     B:NiagaraInt32=/Niagara/DynamicInputs/Subtract/Subtract_Integer(A:NiagaraInt32=50000, B:NiagaraInt32=$Engine.Emitter.NumParticles)
    )
    Use Spawn Probability=false
   ) {enabled=false}
   &SpawnRate(
    SpawnRate=&FloatFromCurve(FloatCurve=DataInterface() {propertyValues {Curve {keys [<at 0.1 1>,<at 0.2 0>]}}}, CurveIndex=$Emitter.Age, Scale Curve=250000)
    Use Spawn Probability=false
   )
  ]
  particleSpawnScript [
   &InitializeParticle(
    Color=&ColorFromCurve(
     ColorCurve=DataInterface() {propertyValues {
       AlphaCurve {keys [<at 0 1 RCIM_Constant>,<at 1 1 RCIM_Constant>]}
       BlueCurve {keys [<at 0 1 RCIM_Constant>,<at 0.3996 0.7905 RCIM_Constant>,<at 0.746 0.9844>]}
       GreenCurve {keys [<at 0 0.2979 RCIM_Constant>,<at 0.3996 0 RCIM_Constant>,<at 0.746 0.8198>]}
       RedCurve {keys [<at 0 0.2865 RCIM_Constant>,<at 0.3996 1 RCIM_Constant>,<at 0.746 0.636>]}
      }}
     CurveIndex=&RandomRangeFloat(Minimum=0, Maximum=1)
    )
    Lifetime=&RandomRangeFloat(Minimum=3, Maximum=6)
    Mesh Scale Mode=::NewEnumerator1
    Mesh Uniform Scale Max=10
    Mesh Uniform Scale Min=5
    Sprite Rotation Mode=::NewEnumerator2
    Sprite Size Mode=::NewEnumerator3
    Uniform Sprite Size=500
    UsePositionOffset=false
    Write Material Random=true
   ) {module=InitializeParticle001}
   &ShapeLocation(MaskSpawnGroup=false, Offset Mode=::NewEnumerator0, Override Seed=false, Sphere Radius=2500, UseOwnerScale=true)
   /SetVariables(Particles.Position=&Add_Vector(
     A=Linked() {linkedVariable {name=Particles.Position, type=NiagaraPosition}}
     B=&MakeVector(X=0, Y=0, Z=&Waveform("Waveform [1]"=::NewEnumerator0, "[1] Frequency"=5, "[1] Amplitude Min/Max"=<0 25>))
    )) {isSetParameters=true, module=SetVariables_#5}
   /SetVariables(Particles.New Vector 3f:Vector3f=<0 0 1>) {isSetParameters=true, module=SetVariables_#6}
   &CameraOffset(Camera Offset Amount=15)
   /SetVariables() {isSetParameters=true, module=SetVariables_#7}
   /SetVariables(Particles.InitialPos:NiagaraPosition=$Particles.Position) {isSetParameters=true, module=SetVariables_#8}
   /Niagara/Modules/Math/PlaceParticlesOnDepthBuffer_GPU(Max Z Depth=10000, Module.Use Max Z Depth=true)
   &InitialMeshOrientation(
    Mesh Axis To Orient=<0 0 1>
    Mesh Orientation Mode=::NewEnumerator2
    Orientation Vector=$Output.PlaceParticlesOnDepthBuffer_GPU.DepthBufferSample_SampleWorldNormal
    Rotation=&RandomRangeVector(Minimum=<0 0 0>, Maximum=<0 0 1>)
    Rotation Coordinate Space=::NewEnumerator2
    Use Rotation Vector=true
   )
   &KillParticles(Kill Particles=&SetBoolByFloatComparison(
     A=&VectorLength(Vector=&Subtract_Vector(A=$Particles.InitialPos, B=Linked() {linkedVariable {name=Particles.Position, type=NiagaraPosition}}))
     B=8000
    )) {module=KillParticles001}
   &KillParticles(Kill Particles=&Boolean_NotOperation(Bool=$Output.PlaceParticlesOnDepthBuffer_GPU.ValidSample)) {module=KillParticles002}
   /Game/Effects/Niagara/Modules/Spawn/FN_Spawn_GBuffer_Color_WithinRangeV2(HueMax=110, HueMin=90)
   &CameraOffset(Camera Offset Amount=10) {module=CameraOffset001}
   &AddVelocity(Velocity=<0 0 0>)
   &KillParticlesInVolume(Invert Volume=true, Sphere Radius=2500) {module=KillParticlesInVolume001}
  ]
  particleUpdateScript [
   &ParticleState(DeltaTime:NiagaraFloat=&MakeCustomFloatFromBool(
     Bool=&SetBoolByFloatComparison(A=&VectorLength(Vector=$Particles.Velocity), B=1)
     True Float=$Engine.DeltaTime
     False Float=0
    ))
   /Niagara/Modules/Update/Forces/PointAttractionForce(
    Attraction Radius=1500
    AttractionStrength:NiagaraFloat=&Multiply_Float(
     A=&FloatFromCurve(
      FloatCurve=DataInterface() {propertyValues {Curve {keys [<at 0 1>,<at 0.25 1>]}}}
      Scale Curve=&RandomRangeFloat(Minimum=-3, Maximum=-5)
     )
     B=1
    )
    AttractorPosition:Vector3f=&Add_Vector(
     A=$Emitter.CameraPosition
     B=&Add_Vector(
      A=&Add_Vector(
       A=&Multiply_VectorByFloat(Vector=<0 0 1>, Float=-100)
       B=&Multiply_VectorByFloat(Vector=$Emitter.CameraForward, Float=-250)
      )
      B=&Multiply_VectorByFloat(Vector=$Emitter.CameraRight, Float=0)
     )
    )
    Falloff Exponent=1
    Module.Kill Within Radius=false
    Module.Use Falloff=false
   )
   /SetVariables(Particles.Normalized Velocity:Vector3f=$Particles.Velocity) {isSetParameters=true, module=SetVariables_#9}
   /SetVariables(Particles.Normalized Velocity:Vector3f=&self:ScratchDynamicInput_03(Default:Vector3f=$Particles.Velocity)) {enabled=false, isSetParameters=true, module=SetVariables_#10}
   /SetVariables(Particles.Is Active:NiagaraBool=&Boolean_OrOperation(
     Bool1=$Particles.Is Active
     Bool2=&SetBoolByFloatComparison(A=&VectorLength(Vector=$Particles.Normalized Velocity), B=1)
    )) {isSetParameters=true, module=SetVariables_#11}
   &CurlNoiseForce(
    Curl Noise Cone Mask Angle=90
    Mask Curl Noise=true
    Noise Frequency=&RandomRangeFloat(Minimum=0.85, Maximum=1)
    Noise Strength=&MakeCustomFloatFromBool(Bool=$Particles.Is Active, True Float=&RandomRangeFloat(Minimum=1500, Maximum=25000), False Float=0)
    Pan Noise=false
    Pan Noise Field=<0 0 0.1>
   )
   &GravityForce(Gravity=&Multiply_VectorByFloat(
     Vector=<0 0 1>
     Float=&MakeCustomFloatFromBool(Bool=$Particles.Is Active, True Float=&RandomRangeFloat(Minimum=500, Maximum=1500), False Float=0)
    ))
   &CurlNoiseForce(
    Noise Frequency=&RandomRangeFloat(Minimum=2, Maximum=3)
    Noise Strength=&MakeCustomFloatFromBool(Bool=$Particles.Is Active, True Float=&RandomRangeFloat(Minimum=5000, Maximum=15000), False Float=0)
    Pan Noise=true
    Pan Noise Field=<0 0 1>
   ) {enabled=false, module=CurlNoiseForce001}
   &PointAttractionForce(
    Attraction Radius=500
    AttractionStrength:NiagaraFloat=&Multiply_Float(
     A=&MakeCustomFloatFromBool(
      Bool=$Particles.Is Active
      True Float=0
      False Float=&FloatFromCurve(FloatCurve=DataInterface() {propertyValues {Curve {keys [<at 0 1>,<at 0.25 0>]}}}, Scale Curve=-8)
     )
     B=1
    )
    AttractorPosition:Vector3f=&self:ScratchDynamicInput(
     CameraPosition:NiagaraPosition=Linked() {linkedVariable {name=Emitter.CameraPosition, type=Vector3f}}
     CameraForward:Vector3f=$Emitter.CameraForward
     HeightOffset:NiagaraFloat=&RandomRangeFloat(Minimum=-100, Maximum=-150)
     CameraForwardOffset:NiagaraFloat=&RandomRangeFloat(Minimum=200, Maximum=250)
    )
    Falloff Exponent=1
    Module.Kill Within Radius=false
    Module.Use Falloff=false
   ) {module=PointAttractionForce001}
   &ScaleMeshSize(Scale Factor=&MakeVector(
     X=1
     Y=1
     Z=&Waveform(
      "Waveform [1]"=::NewEnumerator0
      "[1] Phase Scale"=&RandomRangeFloat(Minimum=0.25, Maximum=3)
      "[1] Frequency"=1
      "[1] Amplitude Min/Max"=<0.7 2>
     )
    )) {module=ScaleMeshSize001}
   &KillParticles(Kill Particles=&self:ScratchDynamicInput_01())
   &Collision(
    Age Colliding Particles=false
    Collision Enabled=&SetBoolByFloatComparison(A=$Particles.Age, B=0.25)
    Enable Rest State=false
    Friction=0
    GPU Collision Type=::NewEnumerator2
    Kill Particles Lodged Within Meshes=false
    KillOnCollision=false
    Max Number Of Collisions=2
    Module.EnableMaxCollisionCount=false
    Module.Randomize Collision Normal Vector=false
    Particle Radius Scale=10
    Randomize Collision Normal=0
    Restitution=1
    Simple Friction=false
    Static Friction=0
    Static Friction Engagement Speed=0
   ) {enabled=false}
   &SolveForcesAndVelocity(Clamp Velocity=true, Limit Acceleration=false, Speed Limit=400) {module=SolveForcesAndVelocity001}
   &CalculateAccurateVelocity()
   &ScaleColor(
    Scale Alpha=&Multiply_Float(
     A=&Multiply_Float(
      A=&Multiply_Float(
       A=1
       B=&FloatFromCurve(FloatCurve=DataInterface() {propertyValues {Curve {keys [<at 0 0>,<at 2 1>]}}}, CurveIndex=$Emitter.Age)
      )
      B=&FloatFromCurve(FloatCurve=DataInterface() {propertyValues {Curve {keys [
           {arriveTangent=9.0725, time=0, value=1}
           {arriveTangent=-0.4023, interpMode=RCIM_Cubic, leaveTangent=-0.4023, tangentMode=RCTM_User, time=0.75, value=0.85}
           {arriveTangent=-0.0405, interpMode=RCIM_Cubic, leaveTangent=-0.0405, tangentMode=RCTM_User, time=1, value=0}
          ]}}})
     )
     B=/Niagara/DynamicInputs/Scalability/ScalabilityDistanceBasedFloat(
      Curve:NiagaraDataInterfaceCurve=DataInterface() {propertyValues {Curve {keys [<at 0.7 1>,<at 1 0>]}}}
      CurveIndex:NiagaraFloat=$Emitter.Scalability.DistanceFraction
      Bias:NiagaraFloat=0
      Scale:NiagaraFloat=1
     )
    )
    Scale RGB=<2 2 2>
    ScaleA=true
    ScaleRGB=true
   )
   &DynamicMaterialParameters(
    Index 0 Param 1=&MakeCustomFloatFromBool(
     Bool=$Particles.Is Active
     True Float=&RandomRangeFloat(Minimum=1.1, Maximum=2)
     False Float=&Multiply_Float(
      A=&Waveform(
       "Waveform [1]"=::NewEnumerator0
       "[1] Frequency"=&RandomRangeFloat(Minimum=0.25, Maximum=1)
       "[1] Amplitude Min/Max"=<0.5 1>
      )
      B=&RandomRangeFloat(Minimum=0.4, Maximum=0.8)
     )
    )
    Index 0 Param 2=&Multiply_Float(
     A=&MakeCustomFloatFromBool(
      Bool=$Particles.Is Active
      True Float=&RandomRangeFloat(Minimum=1, Maximum=2)
      False Float=&RandomRangeFloat(Minimum=0.5, Maximum=1)
     )
     B=&RandomRangeFloat(Minimum=2, Maximum=3)
    )
    Index 0 Param 3=&RandomRangeFloat(Minimum=1, Maximum=2)
    Index 1 Param 2=&RandomRangeFloat(Minimum=0, Maximum=0.15)
    Module.Param0WriteEnabled=true
    Module.Param0WriteEnabled1=true
    Module.Param1WriteEnabled=true
    Module.Param1WriteEnabled1=true
    Module.Param2WriteEnabled=true
    Module.Param2WriteEnabled1=true
    Module.Param3WriteEnabled=true
    Module.Param3WriteEnabled1=true
    Write Parameter Index 1=true
   )
   &ScaleMeshSize(Scale Factor=&MakeVector(
     X=1
     Y=1
     Z=&FloatFromCurve(
      FloatCurve=DataInterface() {propertyValues {Curve {keys [<at 0 1>,<at 0.22 2>,<at 0.25 1>]}}}
      Scale Curve=&Waveform("Waveform [1]"=::NewEnumerator0, "[1] Frequency"=3, "[1] Amplitude Min/Max"=<1 1.5>)
     )
    ))
   &self:ScratchModule_02(ZDampening=1) {module=ScratchModule_02}
   &KillParticlesInVolume(Invert Volume=true, Sphere Radius=8000)
   &self:ScratchModule_04() {module=ScratchModule_04}
  ]
  rendererClasses [&NiagaraMeshRendererProperties, &NiagaraSpriteRendererProperties]
  renderers [
   {rendererClass=&NiagaraMeshRendererProperties, rendererIndex=0}
   {rendererClass=&NiagaraSpriteRendererProperties, rendererIndex=1}
  ]
 }
 Butterflies_High {
  simTarget=GPUComputeSim
  emitterUpdateScript [
   /SetVariables() {isSetParameters=true, module=SetVariables_#3}
   /SetVariables(
    Emitter.CameraForward:Vector3f=&ReturnCameraVector(Camera Property=::NewEnumerator1, Coordinate Space=::Simulation, CameraQuery=$Emitter.CameraRead)
    Emitter.CameraPosition:Vector3f=&ReturnCameraVector(Camera Property=::NewEnumerator0, Coordinate Space=::World, CameraQuery=$Emitter.CameraRead)
    Emitter.CameraRight:Vector3f=&ReturnCameraVector(Camera Property=::NewEnumerator3, Coordinate Space=::Simulation, CameraQuery=$Emitter.CameraRead)
    Emitter.CameraUp:Vector3f=&ReturnCameraVector(Camera Property=::NewEnumerator2, Coordinate Space=::Simulation, CameraQuery=$Emitter.CameraRead)
   ) {isSetParameters=true, module=SetVariables_#4}
   &EmitterState(
    Enable Distance Culling=true
    Life Cycle Mode=::NewEnumerator1
    Loop Behavior=::NewEnumerator1
    Loop Duration Mode=::NewEnumerator1
    Max Distance Response=::SleepAndClearParticles
    MaxDistance=10000
    Module.bApplySpawnCountScaleByDistance=true
    Module.bUseMaxDistance=true
    Module.bUseMinDistance=false
    Reset Age On Awaken=true
    Scalability Mode=::NewEnumerator1
    Scale Spawn Count=true
    Spawn Count Scale By Distance=DataInterface() {propertyValues {Curve {keys [<at 0.5 0>,<at 0.55 1>,<at 1 1>]}}}
    UseLoopDelay=false
   )
   &SpawnPerFrame(
    Spawn:NiagaraBool=&SetBoolByFloatComparison(A=$Emitter.Age, B=1, Comparison Type=::NewEnumerator6)
    Spawn Count:NiagaraInt32=&Min_Int(
     A:NiagaraInt32=500
     B:NiagaraInt32=&Subtract_Integer(A:NiagaraInt32=50000, B:NiagaraInt32=$Engine.Emitter.NumParticles)
    )
    Use Spawn Probability=false
   ) {enabled=false}
   &SpawnRate(
    SpawnRate=&FloatFromCurve(FloatCurve=DataInterface() {propertyValues {Curve {keys [<at 0.1 1>,<at 0.2 0>]}}}, CurveIndex=$Emitter.Age, Scale Curve=250000)
    Use Spawn Probability=false
   )
  ]
  particleSpawnScript [
   &InitializeParticle(
    Color=&ColorFromCurve(
     ColorCurve=DataInterface() {propertyValues {
       AlphaCurve {keys [<at 0 1 RCIM_Constant>,<at 1 1 RCIM_Constant>]}
       BlueCurve {keys [<at 0 1 RCIM_Constant>,<at 0.3996 0.7905 RCIM_Constant>,<at 0.746 0.9844>]}
       GreenCurve {keys [<at 0 0.2979 RCIM_Constant>,<at 0.3996 0 RCIM_Constant>,<at 0.746 0.8198>]}
       RedCurve {keys [<at 0 0.2865 RCIM_Constant>,<at 0.3996 1 RCIM_Constant>,<at 0.746 0.636>]}
      }}
     CurveIndex=&RandomRangeFloat(Minimum=0, Maximum=1)
    )
    Lifetime=&RandomRangeFloat(Minimum=3, Maximum=6)
    Mesh Scale Mode=::NewEnumerator1
    Mesh Uniform Scale Max=10
    Mesh Uniform Scale Min=5
    Sprite Rotation Mode=::NewEnumerator2
    Sprite Size Mode=::NewEnumerator3
    Uniform Sprite Size=500
    UsePositionOffset=false
    Write Material Random=true
   ) {module=InitializeParticle001}
   &ShapeLocation(MaskSpawnGroup=false, Offset Mode=::NewEnumerator0, Override Seed=false, Sphere Radius=2500, UseOwnerScale=true)
   /SetVariables(Particles.Position=&Add_Vector(
     A=Linked() {linkedVariable {name=Particles.Position, type=NiagaraPosition}}
     B=&MakeVector(X=0, Y=0, Z=&Waveform("Waveform [1]"=::NewEnumerator0, "[1] Frequency"=5, "[1] Amplitude Min/Max"=<0 25>))
    )) {isSetParameters=true, module=SetVariables_#5}
   /SetVariables(Particles.New Vector 3f:Vector3f=<0 0 1>) {isSetParameters=true, module=SetVariables_#6}
   &CameraOffset(Camera Offset Amount=15)
   /SetVariables() {isSetParameters=true, module=SetVariables_#7}
   /SetVariables(Particles.InitialPos:NiagaraPosition=$Particles.Position) {isSetParameters=true, module=SetVariables_#8}
   &PlaceParticlesOnDepthBuffer_GPU(Max Z Depth=10000, Module.Use Max Z Depth=true)
   &InitialMeshOrientation(
    Mesh Axis To Orient=<0 0 1>
    Mesh Orientation Mode=::NewEnumerator2
    Orientation Vector=$Output.PlaceParticlesOnDepthBuffer_GPU.DepthBufferSample_SampleWorldNormal
    Rotation=&RandomRangeVector(Minimum=<0 0 0>, Maximum=<0 0 1>)
    Rotation Coordinate Space=::NewEnumerator2
    Use Rotation Vector=true
   )
   &KillParticles(Kill Particles=&SetBoolByFloatComparison(
     A=&VectorLength(Vector=&Subtract_Vector(A=$Particles.InitialPos, B=Linked() {linkedVariable {name=Particles.Position, type=NiagaraPosition}}))
     B=8000
    )) {module=KillParticles001}
   &KillParticles(Kill Particles=&Boolean_NotOperation(Bool=$Output.PlaceParticlesOnDepthBuffer_GPU.ValidSample)) {module=KillParticles002}
   &FN_Spawn_GBuffer_Color_WithinRangeV2(HueMax=110, HueMin=90)
   &CameraOffset(Camera Offset Amount=10) {module=CameraOffset001}
   &AddVelocity(Velocity=<0 0 0>)
   &KillParticlesInVolume(Invert Volume=true, Sphere Radius=2500) {module=KillParticlesInVolume001}
  ]
  particleUpdateScript [
   &ParticleState(DeltaTime:NiagaraFloat=&MakeCustomFloatFromBool(
     Bool=&SetBoolByFloatComparison(A=&VectorLength(Vector=$Particles.Velocity), B=1)
     True Float=$Engine.DeltaTime
     False Float=0
    ))
   &PointAttractionForce(
    Attraction Radius=1500
    AttractionStrength:NiagaraFloat=&Multiply_Float(
     A=&FloatFromCurve(
      FloatCurve=DataInterface() {propertyValues {Curve {keys [<at 0 1>,<at 0.25 1>]}}}
      Scale Curve=&RandomRangeFloat(Minimum=-3, Maximum=-5)
     )
     B=1
    )
    AttractorPosition:Vector3f=&Add_Vector(
     A=$Emitter.CameraPosition
     B=&Add_Vector(
      A=&Add_Vector(
       A=&Multiply_VectorByFloat(Vector=<0 0 1>, Float=-100)
       B=&Multiply_VectorByFloat(Vector=$Emitter.CameraForward, Float=-250)
      )
      B=&Multiply_VectorByFloat(Vector=$Emitter.CameraRight, Float=0)
     )
    )
    Falloff Exponent=1
    Module.Kill Within Radius=false
    Module.Use Falloff=false
   )
   /SetVariables(Particles.Normalized Velocity:Vector3f=$Particles.Velocity) {isSetParameters=true, module=SetVariables_#9}
   /SetVariables(Particles.Normalized Velocity:Vector3f=&self:ScratchDynamicInput_03(Default:Vector3f=$Particles.Velocity)) {enabled=false, isSetParameters=true, module=SetVariables_#10}
   /SetVariables(Particles.Is Active:NiagaraBool=&Boolean_OrOperation(
     Bool1=$Particles.Is Active
     Bool2=&SetBoolByFloatComparison(A=&VectorLength(Vector=$Particles.Normalized Velocity), B=1)
    )) {isSetParameters=true, module=SetVariables_#11}
   &CurlNoiseForce(
    Curl Noise Cone Mask Angle=90
    Mask Curl Noise=true
    Noise Frequency=&RandomRangeFloat(Minimum=0.85, Maximum=1)
    Noise Strength=&MakeCustomFloatFromBool(Bool=$Particles.Is Active, True Float=&RandomRangeFloat(Minimum=1500, Maximum=25000), False Float=0)
    Pan Noise=false
    Pan Noise Field=<0 0 0.1>
   )
   &GravityForce(Gravity=&Multiply_VectorByFloat(
     Vector=<0 0 1>
     Float=&MakeCustomFloatFromBool(Bool=$Particles.Is Active, True Float=&RandomRangeFloat(Minimum=500, Maximum=1500), False Float=0)
    ))
   &CurlNoiseForce(
    Noise Frequency=&RandomRangeFloat(Minimum=2, Maximum=3)
    Noise Strength=&MakeCustomFloatFromBool(Bool=$Particles.Is Active, True Float=&RandomRangeFloat(Minimum=5000, Maximum=15000), False Float=0)
    Pan Noise=true
    Pan Noise Field=<0 0 1>
   ) {enabled=false, module=CurlNoiseForce001}
   &PointAttractionForce(
    Attraction Radius=500
    AttractionStrength:NiagaraFloat=&Multiply_Float(
     A=&MakeCustomFloatFromBool(
      Bool=$Particles.Is Active
      True Float=0
      False Float=&FloatFromCurve(FloatCurve=DataInterface() {propertyValues {Curve {keys [<at 0 1>,<at 0.25 0>]}}}, Scale Curve=-8)
     )
     B=1
    )
    AttractorPosition:Vector3f=&self:ScratchDynamicInput(
     CameraPosition:NiagaraPosition=Linked() {linkedVariable {name=Emitter.CameraPosition, type=Vector3f}}
     CameraForward:Vector3f=$Emitter.CameraForward
     HeightOffset:NiagaraFloat=&RandomRangeFloat(Minimum=-100, Maximum=-150)
     CameraForwardOffset:NiagaraFloat=&RandomRangeFloat(Minimum=200, Maximum=250)
    )
    Falloff Exponent=1
    Module.Kill Within Radius=false
    Module.Use Falloff=false
   ) {module=PointAttractionForce001}
   &ScaleMeshSize(Scale Factor=&MakeVector(
     X=1
     Y=1
     Z=&Waveform(
      "Waveform [1]"=::NewEnumerator0
      "[1] Phase Scale"=&RandomRangeFloat(Minimum=0.25, Maximum=3)
      "[1] Frequency"=1
      "[1] Amplitude Min/Max"=<0.7 2>
     )
    )) {module=ScaleMeshSize001}
   &KillParticles(Kill Particles=&self:ScratchDynamicInput_01())
   &Collision(
    Age Colliding Particles=false
    Collision Enabled=&SetBoolByFloatComparison(A=$Particles.Age, B=0.25)
    Enable Rest State=false
    Friction=0
    GPU Collision Type=::NewEnumerator2
    Kill Particles Lodged Within Meshes=false
    KillOnCollision=false
    Max Number Of Collisions=2
    Module.EnableMaxCollisionCount=false
    Module.Randomize Collision Normal Vector=false
    Particle Radius Scale=10
    Randomize Collision Normal=0
    Restitution=1
    Simple Friction=false
    Static Friction=0
    Static Friction Engagement Speed=0
   ) {enabled=false}
   &SolveForcesAndVelocity(Clamp Velocity=true, Limit Acceleration=false, Speed Limit=400) {module=SolveForcesAndVelocity001}
   &CalculateAccurateVelocity()
   &ScaleColor(
    Scale Alpha=&Multiply_Float(
     A=&Multiply_Float(
      A=&Multiply_Float(
       A=1
       B=&FloatFromCurve(FloatCurve=DataInterface() {propertyValues {Curve {keys [<at 0 0>,<at 2 1>]}}}, CurveIndex=$Emitter.Age)
      )
      B=&FloatFromCurve(FloatCurve=DataInterface() {propertyValues {Curve {keys [
           {arriveTangent=9.0725, time=0, value=1}
           {arriveTangent=-0.4023, interpMode=RCIM_Cubic, leaveTangent=-0.4023, tangentMode=RCTM_User, time=0.75, value=0.85}
           {arriveTangent=-0.0405, interpMode=RCIM_Cubic, leaveTangent=-0.0405, tangentMode=RCTM_User, time=1, value=0}
          ]}}})
     )
     B=&ScalabilityDistanceBasedFloat(
      Curve:NiagaraDataInterfaceCurve=DataInterface() {propertyValues {Curve {keys [<at 0.7 1>,<at 1 0>]}}}
      CurveIndex:NiagaraFloat=$Emitter.Scalability.DistanceFraction
      Bias:NiagaraFloat=0
      Scale:NiagaraFloat=1
     )
    )
    Scale RGB=<2 2 2>
    ScaleA=true
    ScaleRGB=true
   )
   &DynamicMaterialParameters(
    Index 0 Param 1=&MakeCustomFloatFromBool(
     Bool=$Particles.Is Active
     True Float=&RandomRangeFloat(Minimum=1.1, Maximum=2)
     False Float=&Multiply_Float(
      A=&Waveform(
       "Waveform [1]"=::NewEnumerator0
       "[1] Frequency"=&RandomRangeFloat(Minimum=0.25, Maximum=1)
       "[1] Amplitude Min/Max"=<0.5 1>
      )
      B=&RandomRangeFloat(Minimum=0.4, Maximum=0.8)
     )
    )
    Index 0 Param 2=&Multiply_Float(
     A=&MakeCustomFloatFromBool(
      Bool=$Particles.Is Active
      True Float=&RandomRangeFloat(Minimum=1, Maximum=2)
      False Float=&RandomRangeFloat(Minimum=0.5, Maximum=1)
     )
     B=&RandomRangeFloat(Minimum=2, Maximum=3)
    )
    Index 0 Param 3=&RandomRangeFloat(Minimum=1, Maximum=2)
    Index 1 Param 2=&RandomRangeFloat(Minimum=0, Maximum=0.15)
    Module.Param0WriteEnabled=true
    Module.Param0WriteEnabled1=true
    Module.Param1WriteEnabled=true
    Module.Param1WriteEnabled1=true
    Module.Param2WriteEnabled=true
    Module.Param2WriteEnabled1=true
    Module.Param3WriteEnabled=true
    Module.Param3WriteEnabled1=true
    Write Parameter Index 1=true
   )
   &ScaleMeshSize(Scale Factor=&MakeVector(
     X=1
     Y=1
     Z=&FloatFromCurve(
      FloatCurve=DataInterface() {propertyValues {Curve {keys [<at 0 1>,<at 0.22 2>,<at 0.25 1>]}}}
      Scale Curve=&Waveform("Waveform [1]"=::NewEnumerator0, "[1] Frequency"=3, "[1] Amplitude Min/Max"=<1 1.5>)
     )
    ))
   &self:ScratchModule_02(ZDampening=1) {module=ScratchModule_02}
   &KillParticlesInVolume(Invert Volume=true, Sphere Radius=8000)
   &self:ScratchModule_04() {module=ScratchModule_04}
  ]
  rendererClasses [&NiagaraMeshRendererProperties, &NiagaraSpriteRendererProperties]
  renderers [
   {rendererClass=&NiagaraMeshRendererProperties, rendererIndex=0}
   {rendererClass=&NiagaraSpriteRendererProperties, rendererIndex=1}
  ]
 }
}
summary {
 systemName=NS_Llama_ButterFlies
 emitters [
  {
   emitterName=Butterflies_Epic
   simTarget=GPUComputeSim
   rendererClasses [&NiagaraMeshRendererProperties, &NiagaraSpriteRendererProperties]
  }
  {
   emitterName=Butterflies_High
   simTarget=GPUComputeSim
   rendererClasses [&NiagaraMeshRendererProperties, &NiagaraSpriteRendererProperties]
  }
 ]
 userVariables [
  User.BoxSize {type=Vector3f, description="", defaultValue {struct=Vector3f, value=<0 0 0>}}
  User.Radius Mult {type=NiagaraFloat, defaultValue=1, description=""}
  User.Spawn Mult {type=NiagaraFloat, defaultValue=1, description=""}
 ]
}
